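# Animation Blueprint for the DonkeyCrib medium male player body cosmetic that adjusts its pose during the DonkeyCrib emote.
# AnimGraph copies the pose from the attached parent mesh and runs it through the DonkeyCrib Control Rig, passing in the EmoteDonkeyCrib flag.
# Each update checks the owner mesh's anim instance for any of five DonkeyCrib emote montages, then sets that flag and a numeric Is Playing Custom Emote value accordingly.
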AnimBlueprint /BRCosmetics/Characters/Player/Male/Medium/Bodies/M_MED_DonkeyCrib/Meshes/M_MED_DonkeyCrib_AnimBP
target_skeleton=&/M_MED_DonkeyCrib_Skeleton
graphs [
 AnimGraph {nodes [
   Component To Local()
   n1 = Control Rig!(Source=Copy Pose From Mesh() {bUseAttachedParent=true}, Alpha, EmoteDonkeyCrib:bool=Get EmoteDonkeyCrib() -> (EmoteDonkeyCrib:bool)) -> (Pose) {
    alphaBoolBlend {blendOption=Linear}
    controlRigAssetReference {blueprintRigClass=&/M_MED_DonkeyCrib_CtrlRig.M_MED_DonkeyCrib_CtrlRig_C}
   }
   Output Pose(Result=@n1) {layerGroup=None}
  ]}
 EventGraph {nodes [
   Event BlueprintUpdateAnimation()
   > n1 = Sequence()
   n2 = GetAnimInstance(self=GetOwnerSkeletalMeshComponent())
   n3 = OR Boolean!(
    A=Montage_IsPlaying(self=@n2, Montage=/BRCosmetics/Animation/Game/MainPlayer/Emotes/DonkeyCrib/CMM/Emote_DonkeyCrib_CLF_M)
    B=Montage_IsPlaying(self=@n2, Montage=/BRCosmetics/Animation/Game/MainPlayer/Emotes/DonkeyCrib/CMM/Emote_DonkeyCrib_CLM_M)
    C:bool=Montage_IsPlaying(self=@n2, Montage=/BRCosmetics/Animation/Game/MainPlayer/Emotes/DonkeyCrib/CMM/Emote_DonkeyCrib_CMF_M)
    D:bool=Montage_IsPlaying(self=@n2, Montage=/BRCosmetics/Animation/Game/MainPlayer/Emotes/DonkeyCrib/CMM/Emote_DonkeyCrib_CMM_M)
    E:bool=Montage_IsPlaying(self=@n2, Montage=/BRCosmetics/Animation/Game/MainPlayer/Emotes/DonkeyCrib/CMM/Emote_DonkeyCrib_Hero_M)
   ) -> (ReturnValue)
   n4 = Branch(execute=@n1.then_1, Condition=@n3)
   n5 = SelectFloat(A=1, B=0, bPickA=@n3)
   > Set Is Playing Custom Emote(Is Playing Custom Emote:double=@n5) -> (Output_Get:double)
   > Set EmoteDonkeyCrib(EmoteDonkeyCrib:bool=1) -> (Output_Get:bool)
   Set Is Playing Custom Emote(execute=@n4.else, Is Playing Custom Emote=@n5)
   > Set EmoteDonkeyCrib(EmoteDonkeyCrib=false)
  ]}
]
variables [
 EmoteDonkeyCrib {type=boolean}
 Is Playing Custom Emote {type=number}
]
